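# Translucent, unlit, two-sided master material for toon-style smoke shockwaves on Niagara mesh particles and ribbons.
# Opacity subtracts one noise-distorted circle from another, carving a crescent or ring, multiplied by panning Voronoi smoothstep erosion and particle alpha.
# A third distorted circle blends white and grey cel shading, tinted by particle color.
# Dynamic parameters drive erosion, shading inversion and offset, though Shading_Offset_X feeds both axes, leaving Y unused.
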Material /Game/Effects/Niagara/Emitters/Smoke/ToonSmoke/M_Toon_Shockwave_Master
graph {
 properties {
  bAutomaticallySetUsageInEditor=false
  bUsedWithNiagaraMeshParticles=true
  bUsedWithNiagaraRibbons=true
  blendMode=BLEND_Translucent
  shadingModel=MSM_Unlit
  twoSided=true
  usedShadingModels=Unlit
 }
 nodes [
  Constant() {r=-0.1}
  Constant() {r=-0.2}
  n1 = ParticleColor()
  n2 = Multiply(
   A=TextureCoordinate()
   B=AppendVector(
    A=ScalarParameter() {defaultValue=1, parameterName=Base_Distortion_Tile_X}
    B=ScalarParameter() {defaultValue=1, parameterName=Base_Distortion_Tile_Y}
   )
  )
  n3 = AppendVector(
   A=ScalarParameter() {defaultValue=0.01, parameterName=Base_Distortion_Pan_X}
   B=ScalarParameter() {defaultValue=0.1, parameterName=Base_Distortion_Pan_Y}
  )
  n4 = TextureSample(UVs=Add(A=Panner(Coordinate=@n2, Speed=@n3), B=ParticleRandom())) {samplerType=SAMPLERTYPE_Grayscale, texture=/Game/Effects/Fort_Effects/Textures/Fire/T_WispyFireNoise_Soft_M}
  n5 = ScalarParameter() {defaultValue=0.1, parameterName=Base_Distortion_Intensity}
  n6 = TextureSample(UVs=Add(A=TextureCoordinate(), B=Multiply(A=@n4.R, B=@n5))) {samplerSource=SSM_Clamp_WorldGroupSettings, texture=/Game/Athena/Prototype/E3Cutout/FilledCircle}
  n7 = TextureCoordinate() {uTiling=1.25}
  n8 = ScalarParameter() {defaultValue=1, parameterName=Trail_Distortion_Tile_X}
  n9 = ScalarParameter() {defaultValue=0.1, parameterName=Trail_Distortion_Tile_Y}
  n10 = Panner(
   Coordinate=Multiply(A=TextureCoordinate(), B=AppendVector(A=@n8, B=@n9))
   Speed=AppendVector(
    A=ScalarParameter() {defaultValue=0.01, parameterName=Trail_Distortion_Pan_X}
    B=ScalarParameter() {defaultValue=0.02, parameterName=Trail_Distortion_Pan_Y}
   )
  )
  n11 = ParticleRandom()
  n12 = Multiply(
   A=TextureSample(UVs=Add(A=@n10, B=@n11)) {texture=/Game/Creative/PostProcess/Textures/Cell_Noise_T1}.R
   B=ScalarParameter() {defaultValue=0.2, parameterName=Trail_Distortion_Intensity}
  )
  n13 = Constant() {r=-0.2}
  n14 = Constant() {r=-0.3}
  n15 = TextureSample(UVs=Add(A=Add(A=@n7, B=@n12), B=AppendVector(A=@n13, B=@n14))) {samplerSource=SSM_Clamp_WorldGroupSettings, texture=&FilledCircle}
  n16 = Multiply(
   A=TextureCoordinate()
   B=AppendVector(A=ScalarParameter() {defaultValue=0.75, parameterName=Erosion_Tile_X}, B=ScalarParameter() {defaultValue=0.75, parameterName=Erosion_Tile_Y})
  )
  n17 = AppendVector(A=ScalarParameter() {parameterName=Erosion_Pan_X}, B=ScalarParameter() {defaultValue=0.2, parameterName=Erosion_Pan_Y})
  n18 = Multiply(A=TextureCoordinate(), B=ScalarParameter() {defaultValue=1, parameterName=Erosion_Distortion_Tile})
  n19 = AppendVector(
   A=ScalarParameter() {parameterName=Erosion_Distortion_Pan_X}
   B=ScalarParameter() {defaultValue=-0.1, parameterName=Erosion_Distortion_Pan_Y}
  )
  n20 = TextureSample(UVs=Add(A=Panner(Coordinate=@n18, Speed=@n19), B=ParticleRandom())) {samplerType=SAMPLERTYPE_Grayscale, texture=&T_WispyFireNoise_Soft_M}
  n21 = ScalarParameter() {parameterName=Erosion_Distortion_Intensity}
  n22 = TextureSample(UVs=Add(A=Panner(Coordinate=@n16, Speed=@n17), B=Multiply(A=@n20.R, B=@n21))) {samplerType=SAMPLERTYPE_LinearGrayscale, texture=/Game/Packages/Fortress_Sky/SkyDome/Master/T_SkyDome_Cloud_Voroni}
  n23 = ComponentMask(None=TextureCoordinate()) {g=true}
  n24 = DynamicParameter() {defaultValue=<rgba -0.1 -0.2 1 -1>, paramNames [Shading_Offset_X, Shading_Offset_Y, InvertShading, Erosion]}
  n25 = MaterialFunctionCall!(
   Texture=Power(Base=OneMinus(None=@n22.RGB)) {constExponent=1}
   Erosion=Add(A=Power(Base=@n23) {constExponent=1}, B=@n24.Erosion) {constB=0}
   Softness
  ) {materialFunction=/Game/Effects/Niagara/MasterMaterials/Functions/MF_Erosion_Smoothstep_01}
  Multiply(A=@n1.A, B=Multiply(A=Subtract(A=@n6.R, B=@n15.R), B=Multiply(A=@n25.Result)))
  n26 = TextureCoordinate() {uTiling=1.2, vTiling=1.2}
  n27 = ScalarParameter() {defaultValue=1, parameterName=Shading_Distortion_Tile_X}
  n28 = ScalarParameter() {defaultValue=1, parameterName=Shading_Distortion_Tile_Y}
  n29 = Panner(
   Coordinate=Multiply(A=TextureCoordinate(), B=AppendVector(A=@n27, B=@n28))
   Speed=AppendVector(
    A=ScalarParameter() {defaultValue=0.01, parameterName=Shading_Distortion_Pan_X}
    B=ScalarParameter() {defaultValue=0.1, parameterName=Shading_Distortion_Pan_Y}
   )
  )
  n30 = ParticleRandom()
  n31 = Multiply(
   A=TextureSample(UVs=Add(A=@n29, B=@n30)) {samplerType=SAMPLERTYPE_Grayscale, texture=&T_WispyFireNoise_Soft_M}.R
   B=ScalarParameter() {defaultValue=0.1, parameterName=Shading_Distortion_Intensity}
  )
  n32 = DynamicParameter() {defaultValue=<rgba -0.1 -0.2 1 -1>, paramNames [Shading_Offset_X, Shading_Offset_Y, InvertShading, Erosion]}
  n33 = TextureSample(UVs=Add(A=Add(A=@n26, B=@n31), B=AppendVector(A=@n32.Shading_Offset_X, B=@n32.Shading_Offset_X))) {samplerSource=SSM_Clamp_WorldGroupSettings, texture=&FilledCircle}
  n34 = OneMinus(None=@n33.R)
  n35 = DynamicParameter() {defaultValue=<rgba -0.1 -0.2 1 -1>, paramNames [Shading_Offset_X, Shading_Offset_Y, InvertShading, Erosion]}
  Multiply(
   A=LinearInterpolate(
    A=Constant3Vector() {constant=<rgba 1 1 1 1>}
    B=Constant3Vector() {constant=<rgba 0.5 0.5 0.4925 1>}
    Alpha=LinearInterpolate(A=@n33.R, B=@n34, Alpha=@n35.InvertShading)
   )
   B=@n1.RGB
  )
 ]
}
